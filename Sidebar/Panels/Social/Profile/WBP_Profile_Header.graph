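# Sidebar social profile header widget showing avatar, name, subtitle, platform icon and wrapped platform-name tags.
# A Fort Social Item VM initializes the avatar, then supplies a conversation name or a target user for the name/info viewmodels.
# Name setters show nickname over display name, except joinable party cards, which show related names and "+N more".
# Setup Platform Names splits icon markup from text to build tag widgets, with fake entries at design time.
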WidgetBlueprint /Sidebar/Panels/Social/Profile/WBP_Profile_Header
graphs [
 DesignPreviewName {nodes [
   n1 = DesignPreviewName!() -> (then, Name:Text, Subtitle:Text) {is=title K2Node_FunctionEntry}
   > "SetText (Text)"(self=Get Text_Name() -> (Text_Name:Common Text Object Reference), InText=@n1)
   n2 = TextIsEmpty(InText=@n1.Subtitle)
   n3 = Get Text_Subtitle() -> (Text_Subtitle:Common Text Object Reference)
   > SetVisibility(self=@n3, InVisibility=Select(Option 0=SelfHitTestInvisible, Option 1=Collapsed, Index=@n2))
   > "SetText (Text)"(self=@n3, InText=SelectText(B=@n1.Subtitle, bPickA=@n2))
  ]}
 EventGraph {nodes [
   n1 = Event PreConstruct()
   > n2 = Sequence()
   > SetVisibility(
    self=Get WrapBox() -> (WrapBox:Wrap Box Object Reference)
    InVisibility=Select(Option 0=Collapsed, Option 1=HitTestInvisible, Index=Get ShowPlatformNames() -> (ShowPlatformNames:bool))
   )
   Branch(execute=@n2.then_1, Condition=@n1.IsDesignTime)
   > Setup Platform Names!(
    execute
    self
    Platform Name Map:Map of EFortSocialUserPlatformNameTypes Enums to Texts=Make Map!(
     Key 0:EFortSocialUserPlatformNameTypes Enum=Xbox
     Value 0:Text="LOCTEXT(\"<xsx/>FakeName1\")"
     Key 1:EFortSocialUserPlatformNameTypes Enum=Primary
     Value 1:Text="LOCTEXT(\"<console/>FakeDesignName2\")"
     Key 2:EFortSocialUserPlatformNameTypes Enum=Playstation
     Value 2:Text="LOCTEXT(\"<psn/>FakeDesignName3\")"
    ) -> (Map:Map of EFortSocialUserPlatformNameTypes Enums to Texts) {is=title K2Node_MakeMap}
   ) -> (then)
  ]}
 Initialize Header With Fort Social Item VM {nodes [
   n1 = Initialize Header With Fort Social Item VM!() -> (then, FortSocialItemVM:Fort Social Item VM Object Reference) {is=title K2Node_FunctionEntry}
   > Set FortSocialItemVM(FortSocialItemVM:Fort Social Item VM Object Reference=@n1) -> (Output_Get:Fort Social Item VM Object Reference)
   > InitializeWithFortSocialItemVM!(
    execute
    self:WBP Sidebar Player Avatar Object Reference=Get PlayerAvatar() -> (PlayerAvatar:WBP Sidebar Player Avatar Object Reference)
    FortSocialItemVM:Fort Social Item VM Object Reference=@n1
   ) -> (then)
   > n2 = Get!(execute:Exec, self:Fort Social Item VM Object Reference=@n1) -> (then:Exec, else:Exec, TargetConversation:Fort Group Conversation Viewmodel Object Reference)
   > Set Text(Text=Get ConversationName!(self:Fort Text Chat Conversation VM Object Reference=@n2) -> (ConversationName:Text), self=Get Text_Name())
   n3 = Get!(execute=@n2.else, self=@n1) -> (then, else, TargetUser:Fort Social User Weak object)
   > SetSocialUser!(
    execute
    self:Fort Social User Viewmodel Object Reference=[
     Get FortSocialUserNameViewModel() -> (FortSocialUserNameViewModel:Fort Social User Name Viewmodel Object Reference)
     Get FortSocialUserInfoViewModel() -> (FortSocialUserInfoViewModel:Fort Social User Info View Model Object Reference)
    ]
    InSocialUser:Fort Social User Object Reference=@n3
   ) -> (then)
   > n4 = Get!(execute) -> (then, else, FortSocialUserNameViewModel:Fort Social User Name Viewmodel Object Reference)
   > Set PlatformTagDisplayRule!(
    execute
    PlatformTagDisplayRule:EFortSocialPlatformTagDisplayRule Enum=OnlyIfDifferent
    self:Fort Social User Name Viewmodel Object Reference=@n4
   ) -> (then, Output_Get:EFortSocialPlatformTagDisplayRule Enum=Never)
  ]}
 SetNameForJoinablePartyCard {nodes [
   n1 = SetNameForJoinablePartyCard!() -> (then, RelatedUsersDisplayNames:Text) {is=title K2Node_FunctionEntry}
   > Is Valid(InputObject=Get FortSocialItemVM() -> (FortSocialItemVM:Fort Social Item VM Object Reference))
   n2 = Get FortSocialItemVM()
   > Branch(Condition="Equal (Enum)"(
     A:ESocialHeaderItemType Enum=Get HeaderItemType!(self:Fort Social Item VM Object Reference=@n2) -> (HeaderItemType:ESocialHeaderItemType Enum=Unknown)
     B:ESocialHeaderItemType Enum=JoinableNativeEpicParties
    ))
   > "SetText (Text)"(self=Get Text_Name(), InText=@n1)
  ]}
 SetPlayerName {nodes [
   n1 = SetPlayerName!() -> (then, Names:Fort Social User Name Package Structure) {is=title K2Node_FunctionEntry}
   n2 = Get FortSocialItemVM()
   > n3 = Is Valid(InputObject=@n2)
   n4 = Break Fort Social User Name Package!(FortSocialUserNamePackage:"Fort Social User Name Package Structure (by ref)"=@n1) -> (DisplayName:String, Nickname:String, PlatformIconTag:String) {is=title K2Node_BreakStruct}
   n5 = Branch(execute=[@n10.else, @n3.Is Not Valid], Condition=IsEmpty(InString=@n4.Nickname))
   > "SetText (Text)"(self=Get Text_Name(), InText="To Text (String)"(InString=@n4))
   n6 = Get Text_Subtitle()
   > SetVisibility(self=@n6, InVisibility=Collapsed)
   > n7 = "SetText (Text)"(self=@n6, InText="To Text (String)"())
   SetText(
    execute=[@n7.then, @n9.then]
    self=Get RichText_Platform() -> (RichText_Platform:Common Rich Text Block Object Reference)
    InText=Select(
     Option 0:"Text (by ref)"=Format Text!(Format="LOCTEXT(\"<{PlatformIconTag}/>\")", PlatformIconTag:String=@n4.PlatformIconTag) -> (Result)
     Option 1:"Text (by ref)"=GetEmptyText!() -> (ReturnValue:Text)
     Index=IsEmpty(InString=@n4.PlatformIconTag)
    ) -> (ReturnValue:"Text (by ref)")
   )
   "SetText (Text)"(execute=@n5.else, self=Get Text_Name(), InText="To Text (String)"(InString=@n4.Nickname))
   n8 = Get Text_Subtitle()
   > SetVisibility(self=@n8, InVisibility=SelfHitTestInvisible)
   > n9 = "SetText (Text)"(self=@n8, InText="To Text (String)"(InString=@n4))
   n10 = Branch(execute=@n3.Is Valid, Condition="Equal (Enum)"(A=Get HeaderItemType(self=@n2) -> (HeaderItemType=Unknown), B=JoinableNativeEpicParties))
  ]}
 SetUnknownUsersSubtitle {nodes [
   n1 = SetUnknownUsersSubtitle!() -> (then, UnknownUsers:int) {is=title K2Node_FunctionEntry}
   > Is Valid(InputObject=Get FortSocialItemVM())
   n2 = Get FortSocialItemVM()
   > Branch(Condition="Equal (Enum)"(A=Get HeaderItemType(self=@n2) -> (HeaderItemType=Unknown), B=JoinableNativeEpicParties))
   > n3 = Branch(Condition="integer > integer"(A=@n1))
   > SetVisibility(self=Get Text_Subtitle(), InVisibility=SelfHitTestInvisible)
   > "SetText (Text)"(self=Get Text_Subtitle(), InText=Format Text!(Format="LOCTEXT(\"+{N} more\")", N:int=@n1) -> (Result))
   SetVisibility(execute=@n3.else, self=Get Text_Subtitle(), InVisibility=Collapsed)
   > "SetText (Text)"(self=Get Text_Subtitle())
  ]}
 Setup Platform Names {nodes [
   n1 = Setup Platform Names!() -> (then, Platform Name Map:Map of EFortSocialUserPlatformNameTypes Enums to Texts=Primary) {is=title K2Node_FunctionEntry}
   > ClearChildren(self=Get WrapBox())
   > n2 = "For Each Loop (Map)"!(execute:Exec, MapPin:Map of EFortSocialUserPlatformNameTypes Enums to Texts=@n1, BreakPin:Exec) -> (then:Exec, KeyPin:EFortSocialUserPlatformNameTypes Enum, ValuePin:Text, CompletedPin:Exec)
   > n3 = Create Widget!(
    execute:Exec
    Class:User Widget Class Reference=&/WBP_Profile_SocialTag.WBP_Profile_SocialTag_C
    OwningPlayer:Player Controller Object Reference=GetOwningPlayer()
   ) -> (then:Exec, ReturnValue:WBP Profile Social Tag Object Reference)
   n4 = Split(SourceString="To String (Text)"(InText=@n2.ValuePin), InStr="/>", SearchDir=FromStart)
   > SetUsernameText!(execute, self:WBP Profile Social Tag Object Reference=@n3, Text:Text="To Text (String)"(InString=@n4.RightS)) -> (then)
   > SetRichIcon!(execute, self:WBP Profile Social Tag Object Reference=@n3, Text:Text="To Text (String)"(InString=Append(A=@n4, B="/>"))) -> (then)
   > AddChild(self=Get WrapBox(), Content=@n3)
   n5 = Get WrapBox()
   n6 = HasAnyChildren!(self:Panel Widget Object Reference=@n5) -> (ReturnValue:bool)
   n7 = Get ShowPlatformNames()
   SetVisibility(
    execute=@n2.CompletedPin
    self=@n5
    InVisibility=Select(Option 0:ESlateVisibility Enum=Collapsed, Option 1:ESlateVisibility Enum=Visible, Index=AND Boolean(A=@n6, B=@n7)) -> (ReturnValue:ESlateVisibility Enum)
   )
   n8 = Reroute Node(InputPin:WBP UIKit Tag Base Object Reference) -> (OutputPin:WBP UIKit Tag Base Object Reference)
   Set IconMaterial!(execute, IconMaterial:Material Instance Object Reference, self:WBP UIKit Tag Base Object Reference=@n8) -> (then, Output_Get:Material Instance Object Reference)
   > SetLabelText!(execute, self:WBP UIKit Tag Base Object Reference=@n8, Text:Text) -> (then)
   > Set TagSize!(execute, TagSize:E_UI_TagSize Enum=NewEnumerator1, self:WBP UIKit Tag Base Object Reference=@n8) -> (then, Output_Get:E_UI_TagSize Enum=NewEnumerator0)
   > Set TagType!(execute, TagType:E_UI_TagType Enum=NewEnumerator2, self:WBP UIKit Tag Base Object Reference=@n8) -> (then, Output_Get:E_UI_TagType Enum=NewEnumerator0)
   > Set UseIcon!(execute, UseIcon:bool=true, self:WBP UIKit Tag Base Object Reference=@n8) -> (then, Output_Get:bool)
   > SetTagText!(execute, self:WBP UIKit Tag Base Object Reference=@n8, Text:Text) -> (then)
   Reroute Node(InputPin=@n8)
  ]}
]
variables [
 Header Item Type {type=ESocialHeaderItemType, default=Unknown}
 ShowPlatformNames {type=boolean, default=true}
]
widgets {
 info {inheritedWidgetCount=0, parentClass=&UserWidget, rootWidgetClass=&VerticalBox, widgetCount=8}
 widgets [
  VerticalBox_0 {widgetClassPath=&VerticalBox, widget=VerticalBox_0, properties {bOverride_Cursor=true}}
  PlayerAvatar {
   widgetClassPath=/Sidebar/Panels/WBP_Sidebar_PlayerAvatar
   bIsVariable=true
   parent=VerticalBox_0
   widget=PlayerAvatar
   properties {presenceIconScale=1, avatarSize=<104 104>, presenceIconOffset=<0 0>}
   slot {class=VerticalBoxSlot, properties {horizontalAlignment=HAlign_Center, verticalAlignment=VAlign_Top}}
  }
  HorizontalBox_0 {
   widgetClassPath=&HorizontalBox
   parent=VerticalBox_0
   widget=HorizontalBox_0
   slot {class=VerticalBoxSlot, properties {horizontalAlignment=HAlign_Center, padding=<ltrb 0 24 0 0>}}
  }
  Text_Name {
   widgetClassPath=&CommonTextBlock
   bIsVariable=true
   parent=HorizontalBox_0
   widget=Text_Name
   properties {
    clipping=ClipToBounds
    justification=Center
    scrollStyle=&TextScrollStyle-Base_C
    text="{Player Name}"
    textOverflowPolicy=Ellipsis
    font {
     bForceMonospaced=false
     bMaterialIsStencil=false
     fontMaterial=null
     fontObject=&HeadingNow
     letterSpacing=0
     monospacedWidth=1
     size=27
     skewAmount=0
     typefaceFontName=75 Medium
     outlineSettings {
      bApplyOutlineToDropShadows=false
      bMiteredCorners=false
      bSeparateFillAlpha=false
      outlineMaterial=null
      outlineSize=0
      outlineColor=<rgba 0 0 0 1>
     }
    }
    shadowOffset=<0 0>
   }
   slot {class=HorizontalBoxSlot, properties {size {sizeRule=Fill, value=1}}}
  }
  RichText_Platform {
   widgetClassPath=&CommonRichTextBlock
   bIsVariable=true
   parent=HorizontalBox_0
   widget=RichText_Platform
   properties {
    bAutoCollapseWithEmptyText=true
    bIsScrollingEnabled=false
    clipping=ClipToBounds
    mobileTextBlockScale=1.3
    scrollStyle=&TextScrollStyle-Base_C
    textOverflowPolicy=Ellipsis
   }
   slot {class=HorizontalBoxSlot, properties {horizontalAlignment=HAlign_Center, verticalAlignment=VAlign_Center, padding=<ltrb 4 0 0 0>}}
  }
  Text_Subtitle {
   widgetClassPath=&CommonTextBlock
   bIsVariable=true
   parent=VerticalBox_0
   widget=Text_Subtitle
   properties {
    clipping=ClipToBounds
    justification=Center
    scrollStyle=&TextScrollStyle-Base_C
    text="{Player Name}"
    textOverflowPolicy=Ellipsis
    colorAndOpacity {colorUseRule=UseColor_Specified, specifiedColor=<rgba 1 1 1 0.651>}
    font {
     bForceMonospaced=false
     bMaterialIsStencil=false
     fontMaterial=null
     fontObject=&HeadingNow
     letterSpacing=0
     monospacedWidth=1
     size=15.75
     skewAmount=0
     typefaceFontName=75 Medium
     outlineSettings {
      bApplyOutlineToDropShadows=false
      bMiteredCorners=false
      bSeparateFillAlpha=false
      outlineMaterial=null
      outlineSize=0
      outlineColor=<rgba 0 0 0 1>
     }
    }
    shadowOffset=<0 0>
   }
   slot {class=VerticalBoxSlot, properties {padding=<ltrb 0 3 0 0>}}
  }
  WrapBox {
   widgetClassPath=&WrapBox
   bIsVariable=true
   parent=VerticalBox_0
   widget=WrapBox
   properties {bExplicitWrapSize=true, horizontalAlignment=HAlign_Center, innerSlotPadding=<12 12>}
   slot {class=VerticalBoxSlot, properties {padding=<ltrb 30 24 30 0>}}
  }
  Text_FriendRequestSent {
   widgetClassPath=&CommonTextBlock
   bIsVariable=true
   parent=VerticalBox_0
   widget=Text_FriendRequestSent
   properties {
    bIsScrollingEnabled=false
    mobileFontSizeMultiplier=1.33
    scrollStyle=/FortUILibrary/Styles/Typography/Foundation/TextScrollStyle-NoFade.TextScrollStyle-NoFade_C
    text="Friend request sent!"
    textOverflowPolicy=Ellipsis
    visibility=Collapsed
    colorAndOpacity {colorUseRule=UseColor_Specified, specifiedColor=<rgba 0.9301 1 0.0103 1>}
    font {
     bForceMonospaced=false
     bMaterialIsStencil=false
     fontMaterial=/Sidebar/Materials/Fonts/MI_Sidebar_FontColor_Primary
     fontObject=&HeadingNow
     letterSpacing=0
     monospacedWidth=1
     size=15.75
     skewAmount=0
     typefaceFontName=75 Medium
     outlineSettings {
      bApplyOutlineToDropShadows=false
      bMiteredCorners=false
      bSeparateFillAlpha=false
      outlineMaterial=null
      outlineSize=0
      outlineColor=<rgba 0 0 0 1>
     }
    }
    shadowOffset=<0 0>
   }
   slot {class=VerticalBoxSlot, properties {horizontalAlignment=HAlign_Center, padding=<ltrb 0 16 0 24>}}
  }
 ]
}
