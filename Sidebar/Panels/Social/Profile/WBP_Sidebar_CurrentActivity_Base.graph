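# Base CommonUserWidget for the current activity block in the sidebar's social profile panel.
# An island tile image sits beside the activity name and a row with joinable text, divider dot and presence text, framed by thin top and bottom dividers.
# PreConstruct runs SetMobileStyle, which checks IsUsingCompactUI to resize the tile and adjust spacing.
# Two setters control joinable text and dot visibility; no graph here fills in activity data.
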WidgetBlueprint /Sidebar/Panels/Social/Profile/WBP_Sidebar_CurrentActivity_Base
graphs [
 EventGraph {nodes [
   Event PreConstruct()
   > SetMobileStyle() {is=title K2Node_CallFunction}
   Event Construct()
   Event Tick()
  ]}
 SetJoinableDotVisibility {nodes [
   n1 = SetJoinableDotVisibility!() -> (then, Visibility:ESlateVisibility Enum=Visible) {is=title K2Node_FunctionEntry}
   > SetVisibility(self=Get PartyJoinable_Icon_DividerDot() -> (PartyJoinable_Icon_DividerDot:Image Object Reference), InVisibility=@n1)
  ]}
 SetJoinableVisibility {nodes [
   SetJoinableVisibility!() -> (then, Visibility:ESlateVisibility Enum=Visible) {is=title K2Node_FunctionEntry}
   > SetVisibility(
    self=Get TextJoinableStatus() -> (TextJoinableStatus:Common Text Object Reference)
    InVisibility=Get Visibility() -> (Visibility:ESlateVisibility Enum=Visible)
   )
  ]}
 SetMobileStyle {nodes [
   SetMobileStyle() {is=title K2Node_FunctionEntry}
   > Branch(Condition=IsUsingCompactUI())
   n1 = Get IslandTileMaterial() -> (IslandTileMaterial:Image Object Reference)
   > SetVerticalAlignment(self:Grid Slot Object Reference=SlotAsGridSlot(Widget=@n1), InVerticalAlignment=VAlign_Center)
   > SetDesiredSizeOverride!(execute, self:Image Object Reference=@n1, DesiredSize:Vector 2D Structure=(X=138,Y=80)) -> (then)
   n2 = SlotAsVerticalBoxSlot(Widget=Get Image_DividerTop() -> (Image_DividerTop:Image Object Reference))
   n3 = Get!(self:Vertical Box Slot Object Reference=@n2) -> (Padding_Left:float=~, Padding_Top:float=~, Padding_Right:float=~, Padding_Bottom:float=~)
   > SetPadding!(
    execute
    self:Vertical Box Slot Object Reference=@n2
    InPadding_Left:float=@n3
    InPadding_Top:float=12
    InPadding_Right:float=@n3.Padding_Right
    InPadding_Bottom:float=16
   ) -> (then)
   n4 = SlotAsVerticalBoxSlot(Widget=Get Image_DividerBottom() -> (Image_DividerBottom:Image Object Reference))
   n5 = Get(self=@n4) -> (Padding_Left=~, Padding_Top=~, Padding_Right=~, Padding_Bottom=~)
   > SetPadding(self=@n4, InPadding_Left=@n5, InPadding_Top=16, InPadding_Right=@n5.Padding_Right, InPadding_Bottom=12)
   > Set Grid Slot Padding!(
    execute:Exec
    Widget:Widget Object Reference=Get VB_RightColumn() -> (VB_RightColumn:Vertical Box Object Reference)
    Padding_Left:float=24
    Padding_Top:float=0
    Padding_Right:float=0
    Padding_Bottom:float=0
   ) -> (then:Exec) {is=title K2Node_MacroInstance}
   > SetPadding!(execute, self=SlotAsVerticalBoxSlot(Widget=Get GP_Content() -> (GP_Content:Grid Panel Object Reference)), InPadding:Margin Structure) -> (then)
  ]}
]
variables []
widgets {
 info {inheritedWidgetCount=0, parentClass=&CommonUserWidget, rootWidgetClass=&VerticalBox, widgetCount=13}
 widgets [
  VerticalBox_15 {widgetClassPath=&VerticalBox, widget=VerticalBox_15}
  Image_DividerTop {
   widgetClassPath=&Image
   bIsVariable=true
   parent=VerticalBox_15
   widget=Image_DividerTop
   properties {
    visibility=SelfHitTestInvisible
    brush {
     bIsDynamicallyLoaded=false
     drawAs=Image
     imageType=NoImage
     mirroring=NoMirror
     resourceName=None
     resourceObject=null
     tiling=NoTile
     imageSize=<32 2>
     margin=<ltrb 0 0 0 0>
     outlineSettings {
      bUseBrushTransparency=false
      roundingType=HalfHeightRadius
      width=0
      color {colorUseRule=UseColor_Specified, specifiedColor=<rgba 0 0 0 0>}
      cornerRadii=<0 0 0 0>
     }
     tintColor {colorUseRule=UseColor_Specified, specifiedColor=<rgba 1 1 1 1>}
     uVRegion {bIsValid=false, max=<0 0>, min=<0 0>}
    }
    colorAndOpacity=<rgba 1 1 1 0.05>
   }
   slot {class=VerticalBoxSlot, properties {verticalAlignment=VAlign_Center, padding=<ltrb 0 0 0 8>}}
  }
  GP_Content {
   widgetClassPath=&GridPanel
   bIsVariable=true
   parent=VerticalBox_15
   widget=GP_Content
   properties {columnFill [0,1], rowFill [1]}
   slot {class=VerticalBoxSlot, properties {padding=<ltrb 0 10 0 10>}}
  }
  IslandTileMaterial {
   widgetClassPath=&Image
   bIsVariable=true
   parent=GP_Content
   widget=IslandTileMaterial
   properties {
    visibility=SelfHitTestInvisible
    brush {
     bIsDynamicallyLoaded=false
     drawAs=Image
     imageType=FullColor
     mirroring=NoMirror
     resourceName=None
     resourceObject=/Sidebar/Materials/MI_Sidebar_IslandTile
     tiling=NoTile
     imageSize=<106 60>
     margin=<ltrb 0 0 0 0>
     outlineSettings {
      bUseBrushTransparency=false
      roundingType=HalfHeightRadius
      width=0
      color {colorUseRule=UseColor_Specified, specifiedColor=<rgba 0 0 0 0>}
      cornerRadii=<0 0 0 0>
     }
     tintColor {colorUseRule=UseColor_Specified, specifiedColor=<rgba 1 1 1 1>}
     uVRegion {bIsValid=false, max=<0 0>, min=<0 0>}
    }
   }
   slot {class=GridSlot, properties {padding=<ltrb 0 0 18 0>}}
  }
  VB_RightColumn {
   widgetClassPath=&VerticalBox
   bIsVariable=true
   parent=GP_Content
   widget=VB_RightColumn
   slot {class=GridSlot, properties {column=1, verticalAlignment=VAlign_Center}}
  }
  HorizontalStackBox_Name {
   widgetClassPath=&StackBox
   bIsVariable=true
   parent=VB_RightColumn
   widget=HorizontalStackBox_Name
   slot {class=VerticalBoxSlot, properties {verticalAlignment=VAlign_Top, padding=<ltrb 0 0 0 3>}}
  }
  SizeBox_TextClipping {
   widgetClassPath=&SizeBox
   bIsVariable=true
   parent=HorizontalStackBox_Name
   widget=SizeBox_TextClipping
   properties {clipping=ClipToBounds}
   slot {class=StackBoxSlot}
  }
  Text_Group {
   widgetClassPath=&CommonTextBlock
   bIsVariable=true
   parent=SizeBox_TextClipping
   widget=Text_Group
   properties {
    mobileFontSizeMultiplier=1.33
    text=Activity
    textOverflowPolicy=Ellipsis
    font {
     bForceMonospaced=false
     bMaterialIsStencil=false
     fontMaterial=null
     fontObject=&HeadingNow
     letterSpacing=10
     monospacedWidth=1
     size=15.75
     skewAmount=0
     typefaceFontName=75 Medium
     outlineSettings {
      bApplyOutlineToDropShadows=false
      bMiteredCorners=false
      bSeparateFillAlpha=false
      outlineMaterial=/FortUILibrary/Systems/ModularButton/MaterialInstances/MI_UIKit_TextOutline
      outlineSize=0
      outlineColor=<rgba 0 0 0 1>
     }
    }
    shadowOffset=<0 0>
   }
   slot {class=SizeBoxSlot, properties {horizontalAlignment=HAlign_Left, verticalAlignment=VAlign_Center}}
  }
  Grid_ReportingStatus {
   widgetClassPath=&GridPanel
   bIsVariable=true
   parent=VB_RightColumn
   widget=Grid_ReportingStatus
   properties {bOverride_Cursor=true, clipping=ClipToBounds, columnFill [0,0,0,0,1], rowFill [1]}
   slot {class=VerticalBoxSlot, properties {horizontalAlignment=HAlign_Left, verticalAlignment=VAlign_Top}}
  }
  TextJoinableStatus {
   widgetClassPath=&CommonTextBlock
   bIsVariable=true
   parent=Grid_ReportingStatus
   widget=TextJoinableStatus
   properties {
    justification=Center
    mobileFontSizeMultiplier=1.4
    text=Joinable
    colorAndOpacity {colorUseRule=UseColor_Specified, specifiedColor=<rgba 0.1144 0.7682 0.1356 1>}
    font {
     bForceMonospaced=false
     bMaterialIsStencil=false
     fontMaterial=null
     fontObject=&HeadingNow
     letterSpacing=0
     monospacedWidth=1
     size=12.75
     skewAmount=0
     typefaceFontName=75 Medium
     outlineSettings {
      bApplyOutlineToDropShadows=false
      bMiteredCorners=false
      bSeparateFillAlpha=false
      outlineMaterial=&MI_UIKit_TextOutline
      outlineSize=0
      outlineColor=<rgba 0 0 0 1>
     }
    }
    shadowOffset=<0 0>
   }
   slot {class=GridSlot, properties {column=1}}
  }
  PartyJoinable_Icon_DividerDot {
   widgetClassPath=&Image
   bIsVariable=true
   parent=Grid_ReportingStatus
   widget=PartyJoinable_Icon_DividerDot
   properties {
    visibility=SelfHitTestInvisible
    brush {
     bIsDynamicallyLoaded=false
     drawAs=Image
     imageType=FullColor
     mirroring=NoMirror
     resourceName=None
     resourceObject=/FortUILibrary/Foundation/Art_ToolKit/Tools/MI_UI_Circle_ConversationBacchusConfirm
     tiling=NoTile
     imageSize=<6 6>
     margin=<ltrb 0 0 0 0>
     outlineSettings {
      bUseBrushTransparency=false
      roundingType=HalfHeightRadius
      width=0
      color {colorUseRule=UseColor_Specified, specifiedColor=<rgba 0 0 0 0>}
      cornerRadii=<0 0 0 0>
     }
     tintColor {colorUseRule=UseColor_Specified, specifiedColor=<rgba 1 1 1 1>}
     uVRegion {bIsValid=false, max=<0 0>, min=<0 0>}
    }
    colorAndOpacity=<rgba 1 1 1 0.5>
   }
   slot {class=GridSlot, properties {column=2, horizontalAlignment=HAlign_Center, verticalAlignment=VAlign_Center, padding=<ltrb 8 0 8 0>}}
  }
  TextActivityStatus {
   widgetClassPath=&CommonTextBlock
   bIsVariable=true
   parent=Grid_ReportingStatus
   widget=TextActivityStatus
   properties {
    justification=Center
    mobileFontSizeMultiplier=1.4
    scrollStyle=&TextScrollStyle-Base_C
    text=Rich Presence
    textOverflowPolicy=Ellipsis
    colorAndOpacity {colorUseRule=UseColor_Specified, specifiedColor=<rgba 1 1 1 0.65>}
    font {
     bForceMonospaced=false
     bMaterialIsStencil=false
     fontMaterial=null
     fontObject=&HeadingNow
     letterSpacing=0
     monospacedWidth=1
     size=12.75
     skewAmount=0
     typefaceFontName=74 Regular
     outlineSettings {
      bApplyOutlineToDropShadows=false
      bMiteredCorners=false
      bSeparateFillAlpha=false
      outlineMaterial=&MI_UIKit_TextOutline
      outlineSize=0
      outlineColor=<rgba 0 0 0 1>
     }
    }
    shadowOffset=<0 0>
   }
   slot {class=GridSlot, properties {column=4}}
  }
  Image_DividerBottom {
   widgetClassPath=&Image
   bIsVariable=true
   parent=VerticalBox_15
   widget=Image_DividerBottom
   properties {
    visibility=SelfHitTestInvisible
    brush {
     bIsDynamicallyLoaded=false
     drawAs=Image
     imageType=NoImage
     mirroring=NoMirror
     resourceName=None
     resourceObject=null
     tiling=NoTile
     imageSize=<32 2>
     margin=<ltrb 0 0 0 0>
     outlineSettings {
      bUseBrushTransparency=false
      roundingType=HalfHeightRadius
      width=0
      color {colorUseRule=UseColor_Specified, specifiedColor=<rgba 0 0 0 0>}
      cornerRadii=<0 0 0 0>
     }
     tintColor {colorUseRule=UseColor_Specified, specifiedColor=<rgba 1 1 1 1>}
     uVRegion {bIsValid=false, max=<0 0>, min=<0 0>}
    }
    colorAndOpacity=<rgba 1 1 1 0.05>
   }
   slot {class=VerticalBoxSlot, properties {verticalAlignment=VAlign_Center, padding=<ltrb 0 8 0 0>}}
  }
 ]
}
