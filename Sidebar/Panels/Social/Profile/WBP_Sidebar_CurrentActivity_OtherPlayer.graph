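# Widget Blueprint child of WBP_Sidebar_CurrentActivity_Base that presents another player's current activity in the social profile sidebar.
# PreConstruct runs the parent, collapses the island tile image, activity status text and party-joinable divider dot, then centers the reporting-status grid, name row and right column.
# It adds no widgets or variables of its own, only restyling inherited ones.
# SetSocialUser passes the given Fort social user to FortSocialUserInfoViewModel.
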WidgetBlueprint /Sidebar/Panels/Social/Profile/WBP_Sidebar_CurrentActivity_OtherPlayer
graphs [
 EventGraph {nodes [
   n1 = Event PreConstruct()
   > Parent: PreConstruct(IsDesignTime=@n1.IsDesignTime)
   > SetVisibility(
    self=[
     Get IslandTileMaterial() -> (IslandTileMaterial:Image Object Reference)
     Get TextActivityStatus() -> (TextActivityStatus:Common Text Object Reference)
     Get PartyJoinable_Icon_DividerDot() -> (PartyJoinable_Icon_DividerDot:Image Object Reference)
    ]
    InVisibility=Collapsed
   )
   > SetHorizontalAlignment(
    self:Vertical Box Slot Object Reference=SlotAsVerticalBoxSlot(Widget=Get Grid_ReportingStatus() -> (Grid_ReportingStatus:Grid Panel Object Reference))
    InHorizontalAlignment=HAlign_Center
   )
   > SetHorizontalAlignment(
    self=SlotAsVerticalBoxSlot(Widget=Get HorizontalStackBox_Name() -> (HorizontalStackBox_Name:Stack Box Object Reference))
    InHorizontalAlignment=HAlign_Center
   )
   > SetGridSlotAlignment!(
    execute
    Widget:Widget Object Reference=Get VB_RightColumn() -> (VB_RightColumn:Vertical Box Object Reference)
    Horizontal:EHorizontalAlignment Enum=HAlign_Center
    Vertical:EVerticalAlignment Enum=VAlign_Fill
   ) -> (then)
  ]}
 SetSocialUser {nodes [
   n1 = SetSocialUser!() -> (then, Social User:Fort Social User Object Reference) {is=title K2Node_FunctionEntry}
   > SetSocialUser!(
    execute
    self:Fort Social User Viewmodel Object Reference=Get FortSocialUserInfoViewModel() -> (FortSocialUserInfoViewModel:Fort Social User Info View Model Object Reference)
    InSocialUser:Fort Social User Object Reference=@n1
   ) -> (then) {is=title K2Node_CallFunction}
  ]}
]
variables []
widgets {
 info {
  inheritedWidgetCount=0
  parentClass=&/WBP_Sidebar_CurrentActivity_Base.WBP_Sidebar_CurrentActivity_Base_C
  rootWidgetClass=null
  widgetCount=0
 }
 widgets []
}
