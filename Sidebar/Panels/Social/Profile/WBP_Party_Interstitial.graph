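# Sidebar social panel for a party, invite or join request, showing a profile header, current activity, invite banner, interaction buttons and member tree.
# On activation it takes the social item from the current navigation point and activates its interactions.
# Generated buttons map to interaction viewmodels, with certain reject/dismiss actions popping the panel, while member clicks push profile navigation.
# The Fort.UseEpicParties cvar selects banner-matching interaction names; the outer ScrollBox breaks member-list virtualization.
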WidgetBlueprint /Sidebar/Panels/Social/Profile/WBP_Party_Interstitial
graphs [
 BP_GetDesiredFocusTarget {nodes [
   Get Desired Focus Target()
   n1 = Get VB_Interactions() -> (VB_Interactions:Vertical Box Object Reference)
   > n2 = Branch(Condition=HasAnyChildren!(self:Panel Widget Object Reference=@n1) -> (ReturnValue:bool))
   > Return Node!(
    execute
    ReturnValue:Widget Object Reference=GetChildAt!(self:Panel Widget Object Reference=@n1, Index:int=0) -> (ReturnValue:Widget Object Reference)
   )
   n3 = Get TreeView!(self:Fort Polymorphic Tree View Object Reference=Get PartyMembers() -> (PartyMembers:WBP Sidebar Social Tree List Object Reference)) -> (TreeView:Tree View Object Reference)
   n4 = Branch(execute=@n2.else, Condition="integer > integer"(A=GetNumItems!(self:List View Object Reference=@n3) -> (ReturnValue:int=0)))
   n5 = GetIndexForItem!(
    self:List View Object Reference=@n3
    Item:Object Reference=Get Selected Item!(self:List View Object Reference=@n3) -> (ReturnValue:Object Reference) {is=title K2Node_CallFunction}
   ) -> (ReturnValue:int=0)
   > SetSelectedIndex!(
    execute
    self:List View Object Reference=@n3
    Index:int=Select(Option 0:int=0, Option 1:int=@n5, Index="integer >= integer"(A=@n5)) -> (ReturnValue:int)
   ) -> (then)
   > Return Node(ReturnValue=@n3)
   Return Node(execute=@n4.else)
  ]}
 CheckInteractionName {nodes [
   CheckInteractionName!() -> (then, InteractionName:Name) {is=title K2Node_FunctionEntry}
   n1 = Get isEpicParty() -> (isEpicParty:bool)
   n2 = SelectName!(A:Name=DismissEpicPartyJoinRequest, B:Name=DismissJoinRequest, bPickA:bool=@n1) -> (ReturnValue:Name=None)
   n3 = Get InteractionName() -> (InteractionName:Name=None)
   n4 = SelectName(A=AcceptEpicPartyJoinRequest, B=AcceptJoinRequest, bPickA=@n1) -> (ReturnValue=None)
   n5 = Get InteractionName() -> (InteractionName=None)
   n6 = Get TargetConversation!(self:Fort Social Item VM Object Reference=Get FortSocialItemVM() -> (FortSocialItemVM:Fort Social Item VM Object Reference)) -> (TargetConversation:Fort Group Conversation Viewmodel Object Reference)
   n7 = Get isEpicParty()
   n8 = SelectName(
    A=SelectName(A=AcceptEpicPartyConversationInvite, B=AcceptEpicPartyInvite, bPickA=IsValid(Object=@n6)) -> (ReturnValue=None)
    B=AcceptPartyInvite
    bPickA=@n7
   ) -> (ReturnValue=None)
   n9 = Get InteractionName() -> (InteractionName=None)
   n10 = Get TargetConversation(self=Get FortSocialItemVM())
   n11 = SelectName(
    A=SelectName(A=RejectEpicPartyConversationInvite, B=RejectEpicPartyInvite, bPickA=IsValid(Object=@n10)) -> (ReturnValue=None)
    B=RejectPartyInvite
    bPickA=@n7
   ) -> (ReturnValue=None)
   n12 = Get InteractionName() -> (InteractionName=None)
   > Return Node!(
    execute
    IsPartyInvite:bool=OR Boolean(A="Equal (Name)"(A=@n9, B=@n8), B="Equal (Name)"(A=@n12, B=@n11))
    IsJoinRequest:bool=OR Boolean(A="Equal (Name)"(A=@n5, B=@n4), B="Equal (Name)"(A=@n3, B=@n2))
   )
  ]}
 EventGraph {nodes [
   n1 = "Item Edit Nickname (PartyMembers)"!() -> (OutputDelegate, then, SocialUser:Fort Social User Object Reference) {is=title K2Node_ComponentBoundEvent}
   > n2 = Construct Object from Class!(
    execute
    Class=/Sidebar/Shared/Navigation/BP_NavigationPoint.BP_NavigationPoint_C
    self
    SocialUser:Fort Social User Object Reference=@n1.SocialUser
    ConversationId:Name=None
    ConfirmationVM:Sidebar Confirmation Panel VM Object Reference
    ReservedSpots:int=0
    SocialItemVM:Fort Social Item VM Object Reference
    NavigationTag:Gameplay Tag Structure=(TagName="UI.StateTree.TransitionTag.Sidebar.Social.OtherPlayerNickname")
   ) -> (then, ReturnValue:BP Navigation Point Object Reference)
   > PushNavigation!(
    execute
    self:Fort Sidebar Navigation VM Object Reference=Get FortSidebarNavigationVM() -> (FortSidebarNavigationVM:Fort Sidebar Navigation VM Object Reference)
    NavigationPoint:Fort Sidebar Navigation Point Object Reference=@n2
   ) -> (then)
   n3 = "Item Clicked (PartyMembers)"!() -> (OutputDelegate, then, ItemVM:Fort Social Item VM Object Reference) {is=title K2Node_ComponentBoundEvent}
   > n4 = Construct Object from Class(
    Class=&BP_NavigationPoint_C
    SocialUser=Get TargetUser!(self:Fort Social Item VM Object Reference=@n3.ItemVM) -> (TargetUser:Fort Social User Weak object)
    ConversationId=None
    ReservedSpots=0
    SocialItemVM=@n3.ItemVM
    NavigationTag=(TagName="UI.StateTree.TransitionTag.Sidebar.Social.OtherPlayer")
   )
   > PushNavigation(self=Get FortSidebarNavigationVM(), NavigationPoint=@n4)
   n5 = Sidebar Interaction OnClicked!() -> (OutputDelegate, then, Button:Common Button Base Object Reference) {is=title K2Node_CustomEvent}
   > Execute Interaction Button!(execute, self, SocialButton:Common Button Base Object Reference=@n5.Button) -> (then)
   Event Construct()
   > n6 = Construct Object from Class!(execute, Class=/Sidebar/Panels/Social/List/Data/BP_Sidebar_CustomHeaderEntryVM.BP_Sidebar_CustomHeaderEntryVM_C, self) -> (then, ReturnValue:BP Sidebar Custom Header Entry VM Object Reference)
   > Set with Broadcast DisplayName!(execute, DisplayName:Text="LOCTEXT(\"Party Members\")", self:BP Sidebar Custom Header Entry VM Object Reference=@n6) -> (then, Output_Get:Text)
   > Set Party Members Header(Party Members Header:BP Sidebar Custom Header Entry VM Object Reference=@n6) -> (Output_Get:BP Sidebar Custom Header Entry VM Object Reference)
   > n7 = GetConsoleVariableBoolValue!(execute, VariableName:String=Fort.UseEpicParties) -> (then, ReturnValue:bool)
   > Set isEpicParty(isEpicParty:bool=@n7) -> (Output_Get:bool)
   Event On Activated()
   > n8 = GetCurrentNavigationPoint!(execute, self:Fort Sidebar Navigation VM Object Reference=Get FortSidebarNavigationVM()) -> (then, ReturnValue:Fort Sidebar Navigation Point Object Reference)
   > n9 = Cast To BP_NavigationPoint!(execute, Object=@n8) -> (then, CastFailed, AsBP Navigation Point:BP Navigation Point Object Reference) {is=title K2Node_DynamicCast}
   n10 = Get SocialItemVM!(self:BP Navigation Point Object Reference=@n9) -> (SocialItemVM:Fort Social Item VM Object Reference)
   > n11 = Set FortSocialItemVM(FortSocialItemVM:Fort Social Item VM Object Reference=@n10) -> (Output_Get:Fort Social Item VM Object Reference)
   > Initialize Header With Fort Social Item VM!(
    execute
    self:WBP Profile Header Object Reference=Get Profile_Header() -> (Profile_Header:WBP Profile Header Object Reference)
    FortSocialItemVM:Fort Social Item VM Object Reference=@n11
   ) -> (then)
   > SetSocialUser!(
    execute
    self:WBP Sidebar Current Activity Other Player Object Reference=Get CurrentActivity() -> (CurrentActivity:WBP Sidebar Current Activity Other Player Object Reference)
    Social User:Fort Social User Object Reference=Get TargetUser(self=@n10)
   ) -> (then)
   n12 = Get FortSocialItemVM()
   > ActivateInteractionList(self:Fort Social Item VM Object Reference=@n12)
   > RefreshItem(self:Fort Social Item VM Object Reference=@n12)
  ]}
 Execute Interaction Button {nodes [
   n1 = Execute Interaction Button!() -> (then, SocialButton:Common Button Base Object Reference) {is=title K2Node_FunctionEntry}
   > n2 = Cast To WBP_Sidebar_ButtonGeneric!(execute, Object=@n1) -> (then, CastFailed, AsWBP Sidebar Button Generic:WBP Sidebar Button Generic Object Reference) {is=title K2Node_DynamicCast}
   n3 = Find(
    TargetMap:Map of WBP Sidebar Button Generic Object References to Fort Social Interaction VM Object References=Get InteractionsMap() -> (InteractionsMap:Map of WBP Sidebar Button Generic Object References to Fort Social Interaction VM Object References)
    Key:"WBP Sidebar Button Generic Object Reference (by ref)"=@n2
   ) -> (Value:Fort Social Interaction VM Object Reference)
   > Branch(Condition=@n3.ReturnValue)
   n4 = Get InteractionType!(self:Fort Social Interaction VM Object Reference=@n3) -> (InteractionType:ESocialInteractionType Enum=AcceptEpicPartyInvite)
   > Set ShouldCloseCurrentPanel(ShouldCloseCurrentPanel:bool=OR Boolean!(
     A="Equal (Enum)"(A:ESocialInteractionType Enum=@n4, B:ESocialInteractionType Enum=RejectPartyInvite)
     B="Equal (Enum)"(A=@n4, B=RejectEpicPartyInvite)
     C:bool="Equal (Enum)"(A=@n4, B=DismissEpicPartyJoinRequest)
     D:bool="Equal (Enum)"(A=@n4, B=RejectEpicPartyConversationInvite)
    ) -> (ReturnValue)) -> (Output_Get:bool)
   > ExecuteInteraction(self:Fort Social Interaction VM Object Reference=@n3)
   > Branch(Condition=Get ShouldCloseCurrentPanel() -> (ShouldCloseCurrentPanel:bool))
   > PopNavigation(self:Fort Sidebar Navigation VM Object Reference=Get FortSidebarNavigationVM())
  ]}
 NavigateToPartyMember {nodes [
   n1 = Inputs!() -> (execute, self:Fort Social Item VM Object Reference)
   > n2 = Construct Object from Class!(
    execute
    Class=&BP_NavigationPoint_C
    self
    SocialUser:Fort Social User Object Reference
    ConversationId:Name=None
    ConfirmationVM:Sidebar Confirmation Panel VM Object Reference
    ReservedSpots:int=0
    SocialItemVM:Fort Social Item VM Object Reference=@n1
    NavigationTag:Gameplay Tag Structure=(TagName="UI.StateTree.TransitionTag.Sidebar.Social.OtherPlayer")
   ) -> (then, ReturnValue:BP Navigation Point Object Reference)
   > PushNavigation(self=Get FortSidebarNavigationVM(), NavigationPoint=@n2)
   > Outputs!(then:Exec)
  ]}
 Populate Interactions {nodes [
   n1 = Populate Interactions!() -> (then, Interactions:Array of Fort Social Interaction VM Object References) {is=title K2Node_FunctionEntry}
   > ClearChildren(self=Get VB_Interactions())
   > Clear!(execute, TargetMap:Map of WBP Sidebar Button Generic Object References to Fort Social Interaction VM Object References=Get InteractionsMap()) -> (then) {is=title K2Node_CallFunction}
   > SetJoinableVisibility!(execute, self:WBP Sidebar Current Activity Base Object Reference=Get CurrentActivity(), Visibility:ESlateVisibility Enum=Collapsed) -> (then)
   > n2 = For Each Loop(Array:Array of Fort Social Interaction VM Object References=@n1) -> (Array Element:Fort Social Interaction VM Object Reference)
   n3 = Get HasInvite() -> (HasInvite:bool)
   n4 = Get HasJoinRequest() -> (HasJoinRequest:bool)
   SetInviteVisuals!(
    execute=@n2.Completed
    self
    InviteText:Text=SelectText(
     A=SelectText(A="LOCTEXT(\"Invited you to their party!\")", bPickA=@n3)
     B=SelectText(A="LOCTEXT(\"Asked to join your party!\")", bPickA=@n4)
     bPickA=Get HasInvite()
    )
   ) -> (then)
   > RequestRefreshFocus(self:Common Activatable Widget Object Reference)
   n5 = Sequence(execute=@n2.LoopBody)
   > n6 = CheckInteractionName!(execute, self, InteractionName:Name=GetInteractionName!(self:Fort Social Interaction VM Object Reference=@n2) -> (ReturnValue:Name=None)) -> (then, IsPartyInvite:bool, IsJoinRequest:bool) {is=title K2Node_CallFunction}
   > Set HasInvite(HasInvite:bool=OR Boolean!(A=@n6, B=Get HasInvite()) -> (ReturnValue)) -> (Output_Get:bool)
   > Set HasJoinRequest(HasJoinRequest:bool=OR Boolean(A=@n6.IsJoinRequest, B=Get HasJoinRequest())) -> (Output_Get:bool)
   n7 = Create Widget!(
    execute:Exec=@n5.then_1
    Class:User Widget Class Reference=/Sidebar/Buttons/WBP_Sidebar_ButtonGeneric.WBP_Sidebar_ButtonGeneric_C
    OwningPlayer:Player Controller Object Reference=GetOwningPlayer()
    Style:Common Button Style Class Reference=/FortUILibrary/UIKit/Buttons/Blocks/ButtonStyle_UIKit_Empty.ButtonStyle_UIKit_Empty_C
    bApplyAlphaOnDisable:bool=False
    bLocked:bool=False
    bSelectable:bool=True
    bShouldSelectUponReceivingFocus:bool=True
    bInteractableWhenSelected:bool=True
    bToggleable:bool=False
    bTriggerClickedAfterSelection:bool=False
    bDisplayInputActionWhenNotInteractable:bool=false
    bHideInputActionWithKeyboard:bool=False
    bShouldUseFallbackDefaultInputAction:bool=True
    bRequiresHold:bool=False
    HoldData:Common UIHold Data Class Reference=&CommonUI_HoldData_Base_C
    ClickMethod:EButtonClickMethod Enum=DownAndUp
    InputPriority:int=0
    TriggeringInputAction:Data Table Row Handle Structure
   ) -> (then:Exec, ReturnValue:WBP Sidebar Button Generic Object Reference)
   > Add!(
    execute
    TargetMap:Map of WBP Sidebar Button Generic Object References to Fort Social Interaction VM Object References=Get InteractionsMap()
    Key:"WBP Sidebar Button Generic Object Reference (by ref)"=@n7
    Value:"Fort Social Interaction VM Object Reference (by ref)"=@n2
   ) -> (then) {is=title K2Node_CallFunction}
   > SetText!(
    execute
    self:WBP UIKit Button Generic Base Object Reference=@n7
    Text:Text=GetDisplayName!(self:Fort Social Interaction VM Object Reference=@n2) -> (ReturnValue:Text)
   ) -> (then)
   > SetPadding!(
    execute
    self:User Widget Object Reference=@n7
    InPadding_Left:float=0
    InPadding_Top:float=0
    InPadding_Right:float=0
    InPadding_Bottom:float=12
   ) -> (then)
   > Bind Event to On Clicked(self:Common Button Base Object Reference=@n7, Delegate=Create Event()) {is=title K2Node_AddDelegate}
   > AddChild(self=Get VB_Interactions(), Content=@n7)
   n8 = Get InteractionType(self=@n2) -> (InteractionType=AcceptEpicPartyInvite)
   > Branch(Condition=OR Boolean(A="Equal (Enum)"(A=@n8, B=JoinParty), B="Equal (Enum)"(A=@n8, B=JoinNativeEpicParty)))
   > SetJoinableVisibility(self=Get CurrentActivity(), Visibility=SelfHitTestInvisible)
  ]}
 Populate Party Members {nodes [
   n1 = Populate Party Members!() -> (then, Related Users:int) {is=title K2Node_FunctionEntry}
   > Clear(TargetArray:Array of MVVM Base Viewmodel Object References=Get Party Members Array() -> (Party Members Array:Array of MVVM Base Viewmodel Object References)) {is=title K2Node_CallArrayFunction}
   > n2 = For Loop(LastIndex=int - int(A=@n1))
   > n3 = Construct Object from Class!(execute, Class=/Script/SidebarUI.FortSocialItemVM, self) -> (then, ReturnValue:Fort Social Item VM Object Reference)
   > Init!(
    execute
    self:Fort Social Item VM Object Reference=@n3
    InTargetUser:Fort Social User Object Reference=GetRelatedUser!(self:Fort Social Item VM Object Reference=Get FortSocialItemVM(), Index:int=@n2) -> (ReturnValue:Fort Social User Object Reference)
   ) -> (then)
   > Add(
    TargetArray:Array of MVVM Base Viewmodel Object References=Get Party Members Array()
    NewItem:"MVVM Base Viewmodel Object Reference (by ref)"=@n3
   ) {is=title K2Node_CallArrayFunction}
   n4 = Get Party Members Array()
   Branch(execute=@n2.Completed, Condition=Is Not Empty!(TargetArray:Array of MVVM Base Viewmodel Object References=@n4) -> (ReturnValue:bool))
   n5 = Get FortSocialItemVM()
   > n6 = Branch(Condition="integer > integer"(A=GetNumUnknownUsers!(self:Fort Social Item VM Object Reference=@n5) -> (ReturnValue:int=0)))
   > n7 = Construct Object from Class(Class=/Sidebar/Panels/Social/List/Data/BP_SidebarUnknownUsersEntryVM.BP_SidebarUnknownUsersEntryVM_C) -> (ReturnValue:BP Sidebar Unknown Users Entry VM Object Reference)
   > Set Num Unknown Users!(
    execute
    Num Unknown Users:int=GetNumUnknownUsers(self=Get FortSocialItemVM()) -> (ReturnValue=0)
    self:BP Sidebar Unknown Users Entry VM Object Reference=@n7
   ) -> (then, Output_Get:int=0)
   > n8 = Add(TargetArray=Get Party Members Array(), NewItem=@n7)
   n9 = Get Party Members Header() -> (Party Members Header:BP Sidebar Custom Header Entry VM Object Reference)
   Set with Broadcast ChildrenItems!(
    execute=[@n8.then, @n6.else]
    ChildrenItems:Array of MVVM Base Viewmodel Object References=@n4
    self:BP Sidebar Custom Header Entry VM Object Reference=@n9
   ) -> (then, Output_Get:Array of MVVM Base Viewmodel Object References)
   > Set with Broadcast Num Unknown Users!(
    execute
    Num Unknown Users:int=GetNumUnknownUsers(self=Get FortSocialItemVM()) -> (ReturnValue=0)
    self:BP Sidebar Custom Header Entry VM Object Reference=@n9
   ) -> (then, Output_Get:int=0)
   n10 = Get TreeView(self=Get PartyMembers())
   > Set List Items!(
    execute
    self:List View Object Reference=@n10
    InListItems:Array of Object References=Make Array("[0]":Object Reference=@n9) -> (Array:Array of Object References)
   ) -> (then) {is=title K2Node_CallFunction}
   > ExpandAll(self:Tree View Object Reference=@n10)
  ]}
 Set Invite Visuals Visibility {nodes [
   n1 = Inputs() -> (self:Fort Social Interaction VM Object Reference)
   n2 = IsValid(Object=Get TargetConversation(self=Get FortSocialItemVM()))
   n3 = "Equal (Name)"(
    A=GetInteractionName(self=@n1) -> (ReturnValue=None)
    B=SelectName(
     A=SelectName(A=RejectEpicPartyConversationInvite, B=RejectEpicPartyInvite, bPickA=@n2) -> (ReturnValue=None)
     B=RejectPartyInvite
     bPickA=Get isEpicParty()
    ) -> (ReturnValue=None)
   )
   > n4 = Branch(Condition=@n3)
   > n5 = SetVisibility(
    self=Get VB_Invite_Visuals() -> (VB_Invite_Visuals:Horizontal Box Object Reference)
    InVisibility=Select(Option 0:ESlateVisibility Enum=Collapsed, Option 1:ESlateVisibility Enum=SelfHitTestInvisible, Index=@n3) -> (ReturnValue:ESlateVisibility Enum)
   )
   Outputs!(OutputPin:Exec=[@n5.then, @n4.else])
  ]}
 SetInviteVisuals {nodes [
   n1 = SetInviteVisuals!() -> (then, InviteText:Text) {is=title K2Node_FunctionEntry}
   > "SetText (Text)"(self=Get Text_InviteToParty() -> (Text_InviteToParty:Text Block Object Reference), InText=@n1)
   > SetVisibility(self=Get VB_Invite_Visuals(), InVisibility=Select(Option 0=SelfHitTestInvisible, Option 1=Collapsed, Index=TextIsEmpty(InText=@n1)))
  ]}
]
variables [
 InteractionsMap {default {}}
 isEpicParty {type=boolean}
 Party Members Array {type="array<MVVMViewModelBase>", default []}
 Party Members Header {type=BP_Sidebar_CustomHeaderEntryVM_C, default=null}
]
widgets {
 info {
  inheritedWidgetCount=0
  namedSlotCount=1
  parentClass=&CommonActivatableWidget
  rootWidgetClass=/Sidebar/Shared/WBP_Sidebar_PanelLayout.WBP_Sidebar_PanelLayout_C
  widgetCount=10
 }
 widgets [
  PanelLayout {widgetClassPath=/Sidebar/Shared/WBP_Sidebar_PanelLayout, bIsVariable=true, widget=PanelLayout, properties {headingText=""}}
  ScrollBox_NoteBreaksVirtualizingPartyMembersList {
   widgetClassPath=&ScrollBox
   namedSlotHost=PanelLayout
   widget=ScrollBox_NoteBreaksVirtualizingPartyMembersList
   properties {
    clipping=Inherit
    scrollBarVisibility=Collapsed
    scrollWhenFocusChanges=AnimatedScroll
    scrollbarPadding=<ltrb 0 0 0 0>
    scrollbarThickness=<0 0>
    widgetBarStyle {
     thickness=0
     draggedThumbImage {
      bIsDynamicallyLoaded=false
      drawAs=NoDrawType
      imageType=FullColor
      mirroring=NoMirror
      resourceName=../../../Engine/Content/Slate/Common/Scrollbar_Thumb.png
      resourceObject=null
      tiling=NoTile
      imageSize=<0 0>
      margin=<ltrb 0.25 0.25 0.25 0.25>
      outlineSettings {
       bUseBrushTransparency=false
       roundingType=HalfHeightRadius
       width=0
       color {colorUseRule=UseColor_Specified, specifiedColor=<rgba 0 0 0 0>}
       cornerRadii=<0 0 0 0>
      }
      tintColor {colorUseRule=UseColor_Specified, specifiedColor=<rgba 1 1 1 1>}
      uVRegion {bIsValid=false, max=<0 0>, min=<0 0>}
     }
     horizontalBackgroundImage {
      bIsDynamicallyLoaded=false
      drawAs=NoDrawType
      imageType=NoImage
      mirroring=NoMirror
      resourceName=None
      resourceObject=null
      tiling=NoTile
      imageSize=<0 0>
      margin=<ltrb 0 0 0 0>
      outlineSettings {
       bUseBrushTransparency=false
       roundingType=HalfHeightRadius
       width=0
       color {colorUseRule=UseColor_Specified, specifiedColor=<rgba 0 0 0 0>}
       cornerRadii=<0 0 0 0>
      }
      tintColor {colorUseRule=UseColor_Specified, specifiedColor=<rgba 1 1 1 1>}
      uVRegion {bIsValid=false, max=<0 0>, min=<0 0>}
     }
     horizontalBottomSlotImage {
      bIsDynamicallyLoaded=false
      drawAs=NoDrawType
      imageType=FullColor
      mirroring=NoMirror
      resourceName=../../../Engine/Content/Slate/Common/Scrollbar_Background_Horizontal.png
      resourceObject=null
      tiling=NoTile
      imageSize=<8 8>
      margin=<ltrb 0 0 0 0>
      outlineSettings {
       bUseBrushTransparency=false
       roundingType=HalfHeightRadius
       width=0
       color {colorUseRule=UseColor_Specified, specifiedColor=<rgba 0 0 0 0>}
       cornerRadii=<0 0 0 0>
      }
      tintColor {colorUseRule=UseColor_Specified, specifiedColor=<rgba 1 1 1 1>}
      uVRegion {bIsValid=false, max=<0 0>, min=<0 0>}
     }
     horizontalTopSlotImage {
      bIsDynamicallyLoaded=false
      drawAs=NoDrawType
      imageType=FullColor
      mirroring=NoMirror
      resourceName=../../../Engine/Content/Slate/Common/Scrollbar_Background_Horizontal.png
      resourceObject=null
      tiling=NoTile
      imageSize=<8 8>
      margin=<ltrb 0 0 0 0>
      outlineSettings {
       bUseBrushTransparency=false
       roundingType=HalfHeightRadius
       width=0
       color {colorUseRule=UseColor_Specified, specifiedColor=<rgba 0 0 0 0>}
       cornerRadii=<0 0 0 0>
      }
      tintColor {colorUseRule=UseColor_Specified, specifiedColor=<rgba 1 1 1 1>}
      uVRegion {bIsValid=false, max=<0 0>, min=<0 0>}
     }
     hoveredThumbImage {
      bIsDynamicallyLoaded=false
      drawAs=NoDrawType
      imageType=FullColor
      mirroring=NoMirror
      resourceName=../../../Engine/Content/Slate/Common/Scrollbar_Thumb.png
      resourceObject=null
      tiling=NoTile
      imageSize=<0 0>
      margin=<ltrb 0.25 0.25 0.25 0.25>
      outlineSettings {
       bUseBrushTransparency=false
       roundingType=HalfHeightRadius
       width=0
       color {colorUseRule=UseColor_Specified, specifiedColor=<rgba 0 0 0 0>}
       cornerRadii=<0 0 0 0>
      }
      tintColor {colorUseRule=UseColor_Specified, specifiedColor=<rgba 1 1 1 1>}
      uVRegion {bIsValid=false, max=<0 0>, min=<0 0>}
     }
     normalThumbImage {
      bIsDynamicallyLoaded=false
      drawAs=NoDrawType
      imageType=FullColor
      mirroring=NoMirror
      resourceName=../../../Engine/Content/Slate/Common/Scrollbar_Thumb.png
      resourceObject=null
      tiling=NoTile
      imageSize=<0 0>
      margin=<ltrb 0.25 0.25 0.25 0.25>
      outlineSettings {
       bUseBrushTransparency=false
       roundingType=HalfHeightRadius
       width=0
       color {colorUseRule=UseColor_Specified, specifiedColor=<rgba 0 0 0 0>}
       cornerRadii=<0 0 0 0>
      }
      tintColor {colorUseRule=UseColor_Specified, specifiedColor=<rgba 1 1 1 1>}
      uVRegion {bIsValid=false, max=<0 0>, min=<0 0>}
     }
     verticalBackgroundImage {
      bIsDynamicallyLoaded=false
      drawAs=NoDrawType
      imageType=NoImage
      mirroring=NoMirror
      resourceName=None
      resourceObject=null
      tiling=NoTile
      imageSize=<0 0>
      margin=<ltrb 0 0 0 0>
      outlineSettings {
       bUseBrushTransparency=false
       roundingType=HalfHeightRadius
       width=0
       color {colorUseRule=UseColor_Specified, specifiedColor=<rgba 0 0 0 0>}
       cornerRadii=<0 0 0 0>
      }
      tintColor {colorUseRule=UseColor_Specified, specifiedColor=<rgba 1 1 1 1>}
      uVRegion {bIsValid=false, max=<0 0>, min=<0 0>}
     }
     verticalBottomSlotImage {
      bIsDynamicallyLoaded=false
      drawAs=NoDrawType
      imageType=FullColor
      mirroring=NoMirror
      resourceName=../../../Engine/Content/Slate/Common/Scrollbar_Background_Vertical.png
      resourceObject=null
      tiling=NoTile
      imageSize=<8 8>
      margin=<ltrb 0 0 0 0>
      outlineSettings {
       bUseBrushTransparency=false
       roundingType=HalfHeightRadius
       width=0
       color {colorUseRule=UseColor_Specified, specifiedColor=<rgba 0 0 0 0>}
       cornerRadii=<0 0 0 0>
      }
      tintColor {colorUseRule=UseColor_Specified, specifiedColor=<rgba 1 1 1 1>}
      uVRegion {bIsValid=false, max=<0 0>, min=<0 0>}
     }
     verticalTopSlotImage {
      bIsDynamicallyLoaded=false
      drawAs=NoDrawType
      imageType=FullColor
      mirroring=NoMirror
      resourceName=../../../Engine/Content/Slate/Common/Scrollbar_Background_Vertical.png
      resourceObject=null
      tiling=NoTile
      imageSize=<8 8>
      margin=<ltrb 0 0 0 0>
      outlineSettings {
       bUseBrushTransparency=false
       roundingType=HalfHeightRadius
       width=0
       color {colorUseRule=UseColor_Specified, specifiedColor=<rgba 0 0 0 0>}
       cornerRadii=<0 0 0 0>
      }
      tintColor {colorUseRule=UseColor_Specified, specifiedColor=<rgba 1 1 1 1>}
      uVRegion {bIsValid=false, max=<0 0>, min=<0 0>}
     }
    }
    widgetStyle {
     barThickness=0
     bottomShadowBrush {
      bIsDynamicallyLoaded=false
      drawAs=NoDrawType
      imageType=FullColor
      mirroring=NoMirror
      resourceName=../../../Engine/Content/Slate/Common/ScrollBoxShadowBottom.png
      resourceObject=null
      tiling=NoTile
      imageSize=<16 8>
      margin=<ltrb 0.5 0 0.5 1>
      outlineSettings {
       bUseBrushTransparency=false
       roundingType=HalfHeightRadius
       width=0
       color {colorUseRule=UseColor_Specified, specifiedColor=<rgba 0 0 0 0>}
       cornerRadii=<0 0 0 0>
      }
      tintColor {colorUseRule=UseColor_Specified, specifiedColor=<rgba 1 1 1 1>}
      uVRegion {bIsValid=false, max=<0 0>, min=<0 0>}
     }
     horizontalScrolledContentPadding=<ltrb 0 0 1 0>
     leftShadowBrush {
      bIsDynamicallyLoaded=false
      drawAs=NoDrawType
      imageType=FullColor
      mirroring=NoMirror
      resourceName=../../../Engine/Content/Slate/Common/ScrollBoxShadowLeft.png
      resourceObject=null
      tiling=NoTile
      imageSize=<8 16>
      margin=<ltrb 1 0.5 0 0.5>
      outlineSettings {
       bUseBrushTransparency=false
       roundingType=HalfHeightRadius
       width=0
       color {colorUseRule=UseColor_Specified, specifiedColor=<rgba 0 0 0 0>}
       cornerRadii=<0 0 0 0>
      }
      tintColor {colorUseRule=UseColor_Specified, specifiedColor=<rgba 1 1 1 1>}
      uVRegion {bIsValid=false, max=<0 0>, min=<0 0>}
     }
     rightShadowBrush {
      bIsDynamicallyLoaded=false
      drawAs=NoDrawType
      imageType=FullColor
      mirroring=NoMirror
      resourceName=../../../Engine/Content/Slate/Common/ScrollBoxShadowRight.png
      resourceObject=null
      tiling=NoTile
      imageSize=<8 16>
      margin=<ltrb 0 0.5 1 0.5>
      outlineSettings {
       bUseBrushTransparency=false
       roundingType=HalfHeightRadius
       width=0
       color {colorUseRule=UseColor_Specified, specifiedColor=<rgba 0 0 0 0>}
       cornerRadii=<0 0 0 0>
      }
      tintColor {colorUseRule=UseColor_Specified, specifiedColor=<rgba 1 1 1 1>}
      uVRegion {bIsValid=false, max=<0 0>, min=<0 0>}
     }
     topShadowBrush {
      bIsDynamicallyLoaded=false
      drawAs=NoDrawType
      imageType=FullColor
      mirroring=NoMirror
      resourceName=../../../Engine/Content/Slate/Common/ScrollBoxShadowTop.png
      resourceObject=null
      tiling=NoTile
      imageSize=<16 8>
      margin=<ltrb 0.5 1 0.5 0>
      outlineSettings {
       bUseBrushTransparency=false
       roundingType=HalfHeightRadius
       width=0
       color {colorUseRule=UseColor_Specified, specifiedColor=<rgba 0 0 0 0>}
       cornerRadii=<0 0 0 0>
      }
      tintColor {colorUseRule=UseColor_Specified, specifiedColor=<rgba 1 1 1 1>}
      uVRegion {bIsValid=false, max=<0 0>, min=<0 0>}
     }
     verticalScrolledContentPadding=<ltrb 0 0 0 1>
    }
   }
  }
  Content {
   widgetClassPath=&VerticalBox
   parent=ScrollBox_NoteBreaksVirtualizingPartyMembersList
   widget=Content
   slot {class=ScrollBoxSlot, properties {size {sizeRule=Fill, value=1}}}
  }
  Profile_Header {
   widgetClassPath=&/WBP_Profile_Header
   bIsVariable=true
   parent=Content
   widget=Profile_Header
   properties {showPlatformNames=false}
   slot {class=VerticalBoxSlot, properties {padding=<ltrb 0 30 0 0>}}
  }
  VB_Invite_Visuals {
   widgetClassPath=&HorizontalBox
   bIsVariable=true
   parent=Content
   widget=VB_Invite_Visuals
   properties {visibility=Collapsed}
   slot {class=VerticalBoxSlot, properties {horizontalAlignment=HAlign_Center, verticalAlignment=VAlign_Center, padding=<ltrb 0 18 0 18>}}
  }
  Icon_InviteToParty {
   widgetClassPath=&Image
   bIsVariable=true
   parent=VB_Invite_Visuals
   widget=Icon_InviteToParty
   properties {brush {
     bIsDynamicallyLoaded=false
     drawAs=Image
     imageType=FullColor
     mirroring=NoMirror
     resourceName=None
     resourceObject=/Sidebar/Materials/Icons/MI_Sidebar_Icon_PartyInvite
     tiling=NoTile
     imageSize=<24 24>
     margin=<ltrb 0 0 0 0>
     outlineSettings {
      bUseBrushTransparency=false
      roundingType=HalfHeightRadius
      width=0
      color {colorUseRule=UseColor_Specified, specifiedColor=<rgba 0 0 0 0>}
      cornerRadii=<0 0 0 0>
     }
     tintColor {colorUseRule=UseColor_Specified, specifiedColor=<rgba 1 1 1 1>}
     uVRegion {bIsValid=false, max=<0 0>, min=<0 0>}
    }}
   slot {
    class=HorizontalBoxSlot
    properties {horizontalAlignment=HAlign_Center, verticalAlignment=VAlign_Center, padding=<ltrb 0 0 4 0>, size {sizeRule=Fill, value=1}}
   }
  }
  Text_InviteToParty {
   widgetClassPath=&UEFN_TextBlock
   bIsVariable=true
   parent=VB_Invite_Visuals
   widget=Text_InviteToParty
   properties {
    text="Invited you to their party!"
    visibility=SelfHitTestInvisible
    colorAndOpacity {colorUseRule=UseColor_Specified, specifiedColor=<rgba 0.9301 1 0.0103 1>}
    font {
     bForceMonospaced=false
     bMaterialIsStencil=false
     fontMaterial=null
     fontObject=&HeadingNow
     letterSpacing=0
     monospacedWidth=1
     size=15.75
     skewAmount=0
     typefaceFontName=75 Medium
     outlineSettings {
      bApplyOutlineToDropShadows=false
      bMiteredCorners=false
      bSeparateFillAlpha=false
      outlineMaterial=null
      outlineSize=0
      outlineColor=<rgba 0 0 0 1>
     }
    }
   }
   slot {class=HorizontalBoxSlot, properties {verticalAlignment=VAlign_Center}}
  }
  CurrentActivity {
   widgetClassPath=&/WBP_Sidebar_CurrentActivity_OtherPlayer
   bIsVariable=true
   parent=Content
   widget=CurrentActivity
   slot {class=VerticalBoxSlot, properties {padding=<ltrb 6 0 6 0>}}
  }
  VB_Interactions {
   widgetClassPath=&VerticalBox
   bIsVariable=true
   parent=Content
   widget=VB_Interactions
   slot {class=VerticalBoxSlot, properties {padding=<ltrb 24 24 24 0>}}
  }
  PartyMembers {
   widgetClassPath=/Sidebar/Panels/Social/List/WBP_Sidebar_SocialTreeList
   bIsVariable=true
   parent=Content
   widget=PartyMembers
   slot {class=VerticalBoxSlot, properties {padding=<ltrb 0 24 0 0>}}
  }
 ]
}
