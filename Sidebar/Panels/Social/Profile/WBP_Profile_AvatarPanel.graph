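# Common UI activatable widget for the "Change Avatar" panel in the sidebar's social profile section.
# A root grid holds the shared sidebar panel layout, with a blank heading and an empty grid in its named slot, plus a padded "Change Avatar" text label beside it as a sibling.
# The widget activates itself on Construct and contains no avatar selection logic.
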WidgetBlueprint /Sidebar/Panels/Social/Profile/WBP_Profile_AvatarPanel
graphs [EventGraph {nodes [
   Event Construct()
   > ActivateWidget()
   Event PreConstruct()
   Event On Activated()
  ]}]
variables []
widgets {
 info {
  inheritedWidgetCount=0
  namedSlotCount=1
  parentClass=&CommonActivatableWidget
  rootWidgetClass=&GridPanel
  widgetCount=4
 }
 widgets [
  Avatar_GridRoot {widgetClassPath=&GridPanel, widget=Avatar_GridRoot, properties {columnFill [1], rowFill [1]}}
  PanelLayout {
   widgetClassPath=/Sidebar/Shared/WBP_Sidebar_PanelLayout
   bIsVariable=true
   parent=Avatar_GridRoot
   widget=PanelLayout
   properties {headingText=""}
   slot {class=GridSlot}
  }
  Layout {widgetClassPath=&GridPanel, namedSlotHost=PanelLayout, widget=Layout, properties {columnFill [1], rowFill [0,1]}}
  CommonTextBlock_117 {
   widgetClassPath=&CommonTextBlock
   parent=Avatar_GridRoot
   widget=CommonTextBlock_117
   properties {
    style=&TextStyle-Base_C
    text=Change Avatar
    font {
     bForceMonospaced=false
     bMaterialIsStencil=false
     fontMaterial=null
     fontObject=&NotoSans
     letterSpacing=0
     monospacedWidth=1
     size=24
     skewAmount=0
     typefaceFontName=NotoSans-Regular
     outlineSettings {
      bApplyOutlineToDropShadows=false
      bMiteredCorners=false
      bSeparateFillAlpha=false
      outlineMaterial=null
      outlineSize=0
      outlineColor=<rgba 0 0 0 1>
     }
    }
    shadowOffset=<0 0>
   }
   slot {class=GridSlot, properties {padding=<ltrb 30 160 0 0>}}
  }
 ]
}
