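# WBP_Profile_Settings is a placeholder settings panel under the sidebar's Social Profile panels, built on CommonActivatableWidget.
# A single-cell fill GridPanel centers a CommonTextBlock reading "Communication Settings" and "Coming Soon".
# It has no logic or variables, as its PreConstruct, Construct and Tick events are empty.
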WidgetBlueprint /Sidebar/Panels/Social/Profile/WBP_Profile_Settings
graphs [EventGraph {nodes [
   Event PreConstruct()
   Event Construct()
   Event Tick()
  ]}]
variables []
widgets {
 info {inheritedWidgetCount=0, parentClass=&CommonActivatableWidget, rootWidgetClass=&GridPanel, widgetCount=2}
 widgets [
  CommSettings_RootGrid {widgetClassPath=&GridPanel, widget=CommSettings_RootGrid, properties {columnFill [1], rowFill [1]}}
  Text_ComingSoon {
   widgetClassPath=&CommonTextBlock
   parent=CommSettings_RootGrid
   widget=Text_ComingSoon
   properties {
    text="Communication Settings\r\n\r\nComing Soon"
    font {
     bForceMonospaced=false
     bMaterialIsStencil=false
     fontMaterial=null
     fontObject=&NotoSans
     letterSpacing=0
     monospacedWidth=1
     size=24
     skewAmount=0
     typefaceFontName=NotoSans-Regular
     outlineSettings {
      bApplyOutlineToDropShadows=false
      bMiteredCorners=false
      bSeparateFillAlpha=false
      outlineMaterial=null
      outlineSize=0
      outlineColor=<rgba 0 0 0 1>
     }
    }
    shadowOffset=<0 0>
   }
   slot {class=GridSlot, properties {horizontalAlignment=HAlign_Center, verticalAlignment=VAlign_Center}}
  }
 ]
}
