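# Sidebar button widget for online status, derived from the generic sidebar button, showing a material-driven status icon (offline by default) with a text label.
# On PreConstruct it applies a compact style when compact UI is in use, adjusting the icon size, icon top padding and bottom text padding.
# SetPreviewData casts a view model to UIKitButtonViewModel, then feeds its label into the button text and its icon material into the image brush through SetLocalIcon.
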WidgetBlueprint /Sidebar/Buttons/WBP_Sidebar_ButtonOnlineStatus
graphs [
 EventGraph {nodes [
   n1 = Event PreConstruct()
   > Parent: PreConstruct(IsDesignTime=@n1.IsDesignTime)
   > SetCompactUIStyle(self:WBP Sidebar Button Generic Object Reference) {is=title K2Node_CallFunction}
   Event Construct()
   > Parent: Construct()
  ]}
 SetCompactUIStyle {nodes [
   SetCompactUIStyle() {is=title K2Node_FunctionEntry}
   > Branch(Condition=IsUsingCompactUI())
   n1 = Get Image_Icon() -> (Image_Icon:Image Object Reference)
   > SetDesiredSizeOverride!(execute, self:Image Object Reference=@n1, DesiredSize:Vector 2D Structure=(X=20,Y=20)) -> (then)
   n2 = SlotAsOverlaySlot(Widget=@n1)
   n3 = Get!(self:Overlay Slot Object Reference=@n2) -> (Padding_Left:float=~, Padding_Top:float=~, Padding_Right:float=~, Padding_Bottom:float=~)
   > SetPadding!(
    execute
    self:Overlay Slot Object Reference=@n2
    InPadding_Left:float=@n3
    InPadding_Top:float=18
    InPadding_Right:float=@n3.Padding_Right
    InPadding_Bottom:float=@n3.Padding_Bottom
   ) -> (then)
   n4 = Get!() -> (TextPadding_Left:float=~, TextPadding_Top:float=~, TextPadding_Right:float=~, TextPadding_Bottom:float=~)
   > SetTextPadding!(
    execute
    self:WBP UIKit Button Generic Base Object Reference
    TextPadding_Left:float=@n4
    TextPadding_Top:float=@n4.TextPadding_Top
    TextPadding_Right:float=@n4.TextPadding_Right
    TextPadding_Bottom:float=14
   ) -> (then)
   > UpdateText(self:WBP UIKit Button Generic Base Object Reference)
  ]}
 SetLocalIcon {nodes [
   n1 = SetLocalIcon!() -> (then, Material:Material Interface Object Reference) {is=title K2Node_FunctionEntry}
   > SetBrushFromMaterial(self=Get Image_Icon(), Material=@n1)
  ]}
 SetPreviewData {nodes [
   n1 = SetPreviewData!() -> (then, VM:MVVM Base Viewmodel Object Reference) {is=title K2Node_FunctionEntry}
   > n2 = Cast To UIKitButtonViewModel!(execute, Object=@n1) -> (then, CastFailed, AsUI Kit Button View Model:UI Kit Button View Model Object Reference) {is=title K2Node_DynamicCast}
   > SetText!(
    execute
    self:WBP UIKit Button Generic Base Object Reference
    Text:Text=Get Label!(self:UI Kit Button View Model Object Reference=@n2) -> (Label:Text)
   ) -> (then)
   > SetLocalIcon!(
    execute
    self
    Material:Material Interface Object Reference=Get IconMaterial!(self:UI Kit Button View Model Object Reference=@n2) -> (IconMaterial:Material Instance Object Reference)
   ) -> (then) {is=title K2Node_CallFunction}
  ]}
]
variables []
widgets {
 info {
  inheritedWidgetCount=0
  namedSlotCount=1
  parentClass=&/WBP_Sidebar_ButtonGeneric.WBP_Sidebar_ButtonGeneric_C
  rootWidgetClass=null
  widgetCount=2
 }
 widgets [
  Overlay_30 {widgetClassPath=&Overlay, widget=Overlay_30}
  Image_Icon {
   widgetClassPath=&Image
   bIsVariable=true
   parent=Overlay_30
   widget=Image_Icon
   properties {brush {
     bIsDynamicallyLoaded=false
     drawAs=Image
     imageType=FullColor
     mirroring=NoMirror
     resourceName=None
     resourceObject=/Sidebar/Materials/Icons/MI_Sidebar_StatusOffline
     tiling=NoTile
     imageSize=<15 15>
     margin=<ltrb 0 0 0 0>
     outlineSettings {
      bUseBrushTransparency=false
      roundingType=HalfHeightRadius
      width=0
      color {colorUseRule=UseColor_Specified, specifiedColor=<rgba 0 0 0 0>}
      cornerRadii=<0 0 0 0>
     }
     tintColor {colorUseRule=UseColor_Specified, specifiedColor=<rgba 1 1 1 1>}
     uVRegion {bIsValid=false, max=<0 0>, min=<0 0>}
    }}
   slot {class=OverlaySlot, properties {horizontalAlignment=HAlign_Center, padding=<ltrb 0 17 0 0>}}
  }
 ]
}
