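# Sidebar social profile widget for the local player's current activity, derived from WBP_Sidebar_CurrentActivity_Base with no widgets of its own.
# SetSocialUser forwards the user to PartyMemberCurrentIslandVM, and a valid island view model triggers an image fetch.
# The image loading state drives a dynamic material on the island tile, toggling loading and error parameters and applying the thumbnail once loaded.
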WidgetBlueprint /Sidebar/Panels/Social/Profile/WBP_Sidebar_CurrentActivity_LocalPlayer
graphs [
 EventGraph {nodes [
   n1 = Event PreConstruct()
   > Parent: PreConstruct(IsDesignTime=@n1.IsDesignTime)
  ]}
 OnCurrentIslandVMSet {nodes [
   n1 = OnCurrentIslandVMSet!() -> (then, Island VM:Discover Activity View Model Object Reference) {is=title K2Node_FunctionEntry}
   > Is Valid(InputObject=@n1)
   > FetchImages(self:Fort Discover Link Entry Tile Item VM Object Reference=Get CurrentIslandVM!(self:Party Member Current Island View Model Object Reference=Get PartyMemberCurrentIslandVM() -> (PartyMemberCurrentIslandVM:Party Member Current Island View Model Object Reference)) -> (CurrentIslandVM:Discover Activity View Model Object Reference))
  ]}
 SetCurrentIslandImageLoadingState {nodes [
   SetCurrentIslandImageLoadingState!() -> (then, IslandImageState:EActivityImageLoadingState Enum=Unset) {is=title K2Node_FunctionEntry}
   n1 = Get IslandTileMaterial() -> (IslandTileMaterial:Image Object Reference)
   > Set IslandTile MID(IslandTile MID:Material Instance Dynamic Object Reference=GetDynamicMaterial!(Brush:"Slate Brush Structure (by ref)"=Get Brush!(self:Image Object Reference=@n1) -> (Brush:Slate Brush Structure)) -> (ReturnValue)) -> (Output_Get:Material Instance Dynamic Object Reference)
   > n2 = Switch on EActivityImageLoadingState!(execute, Selection:EActivityImageLoadingState Enum=Get IslandImageState() -> (IslandImageState:EActivityImageLoadingState Enum=Unset)) -> (Unset:Exec, Loading:Exec, Loaded:Exec, Error:Exec) {is=title K2Node_SwitchEnum}
   SetScalarParameterValue(
    execute=@n2.Loaded
    self=Get IslandTile MID() -> (IslandTile MID:Material Instance Dynamic Object Reference)
    ParameterName=IsLoading
    Value=0
   )
   n3 = Get PartyMemberCurrentIslandVM()
   > SetTextureParameterValue(
    self=Get IslandTile MID()
    ParameterName=Thumbnail
    Value=GetImageTexture!(self:Fort Discover Link Entry Tile Item VM Object Reference=Get CurrentIslandVM(self=@n3)) -> (ReturnValue:Texture Object Reference)
   )
   SetScalarParameterValue(execute=@n2.Loading, self=Get IslandTile MID(), ParameterName=IsLoading, Value=1)
   SetScalarParameterValue(execute=@n2.Error, self=Get IslandTile MID(), ParameterName=IsLoading, Value=0)
   > SetScalarParameterValue(self=Get IslandTile MID(), ParameterName=IsError, Value=1)
  ]}
 SetSocialUser {nodes [
   n1 = SetSocialUser!() -> (then, Social User:Fort Social User Object Reference) {is=title K2Node_FunctionEntry}
   > SetSocialUser!(
    execute
    self:Fort Social User Viewmodel Object Reference=Get PartyMemberCurrentIslandVM()
    InSocialUser:Fort Social User Object Reference=@n1
   ) -> (then) {is=title K2Node_CallFunction}
  ]}
]
variables [IslandTile MID {type=MaterialInstanceDynamic}]
widgets {
 info {
  inheritedWidgetCount=0
  parentClass=&/WBP_Sidebar_CurrentActivity_Base.WBP_Sidebar_CurrentActivity_Base_C
  rootWidgetClass=null
  widgetCount=0
 }
 widgets []
}
