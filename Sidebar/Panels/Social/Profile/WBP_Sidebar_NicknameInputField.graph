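# Sidebar social-profile widget, built on the generic sidebar button, for entering, changing or removing a nickname.
# Typing runs validation through the nickname view model, updating a character counter or showing red errors; empty text counts as valid.
# Enter fires OnChangedNickname, while TryChangeNickname revalidates before committing to the view model.
# Bind functions refresh text and Add/Remove hint and disable input for Cabined or Minor accounts.
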WidgetBlueprint /Sidebar/Panels/Social/Profile/WBP_Sidebar_NicknameInputField
graphs [
 BindHandleError {nodes [
   n1 = BindHandleError!() -> (then, Error:ENicknamingBlockedReason Enum=None) {is=title K2Node_FunctionEntry}
   > n2 = Switch on ENicknamingBlockedReason!(execute, Selection:ENicknamingBlockedReason Enum=@n1) -> (None:Exec, Cabined:Exec, Minor:Exec, Sanitized:Exec, InvalidLength:Exec, InvalidFormat:Exec, NoUser:Exec) {is=title K2Node_SwitchEnum}
   n3 = Self-Reference()
   n4 = Get EditableText() -> (EditableText:Fort Mobile Editable Text Object Reference)
   SetIsEnabled(execute=[@n2.Cabined, @n2.Minor], self=[@n3, @n4], bInIsEnabled=false)
   SetIsEnabled(execute=[@n2.InvalidFormat, @n2.InvalidLength, @n2.None, @n2.Sanitized], self=[@n3, @n4], bInIsEnabled=true)
  ]}
 BindSetNickname {nodes [
   n1 = BindSetNickname!() -> (then, Text:Text) {is=title K2Node_FunctionEntry}
   n2 = Get EditableText()
   > "SetText (Editable Text)"!(execute, self:Editable Text Object Reference=@n2, InText:Text=@n1) -> (then)
   n3 = Get FortNicknameVM() -> (FortNicknameVM:Nickname VM Object Reference)
   > SetHintText!(
    execute
    self:Editable Text Object Reference=@n2
    InHintText:Text=Select(
     Option 0:Text="LOCTEXT(\"Add Nickname\")"
     Option 1:Text="LOCTEXT(\"Remove Nickname\")"
     Index=HasNickname!(self:Nickname VM Object Reference=@n3) -> (ReturnValue:bool)
    ) -> (ReturnValue:Text)
   ) -> (then)
  ]}
 EventGraph {nodes [
   n1 = OnEditTextChanged!() -> (OutputDelegate, then, Text:"Text (by ref)") {is=title K2Node_CustomEvent}
   > ValidateNickname!(execute, self, Text:"Text (by ref)"=@n1.Text) -> (then, Valid:bool)
   n2 = OnEditTextCommited!() -> (OutputDelegate, then, Text:"Text (by ref)", CommitMethod:ETextCommit Enum=Default) {is=title K2Node_CustomEvent}
   > n3 = Switch on ETextCommit!(execute, Selection:ETextCommit Enum=@n2.CommitMethod) -> (Default:Exec, OnEnter:Exec, OnUserMovedFocus:Exec, OnCleared:Exec) {is=title K2Node_SwitchEnum}
   Call On Changed Nickname!(execute:Exec=@n3.OnEnter, self:WBP Sidebar Nickname Input Field Object Reference) -> (then:Exec) {is=title K2Node_CallDelegate}
   n4 = Event PreConstruct()
   > Parent: PreConstruct(IsDesignTime=@n4.IsDesignTime)
   > Set Visuals!(execute:Exec) -> (then:Exec) {is=title K2Node_MacroInstance}
   Event Construct()
   > Parent: Construct()
   n5 = Create Event()
   n6 = Get EditableText()
   > "Bind Event to OnTextChanged (Editable Text)"(self:Editable Text Object Reference=@n6, Delegate=@n5) {is=title K2Node_AddDelegate}
   n7 = Create Event()
   > "Bind Event to OnTextCommitted (Editable Text)"(self:Editable Text Object Reference=@n6, Delegate=@n7) {is=title K2Node_AddDelegate}
   Event Destruct()
   > Parent: Destruct()
   > "Unbind Event from OnTextChanged (Editable Text)"(self:Editable Text Object Reference=@n6, Delegate=@n5) {is=title K2Node_RemoveDelegate}
   > "Unbind Event from OnTextCommitted (Editable Text)"(self:Editable Text Object Reference=@n6, Delegate=@n7) {is=title K2Node_RemoveDelegate}
  ]}
 OnChangedNickname {nodes [OnChangedNickname() {is=title K2Node_FunctionEntry}]}
 SetVM {nodes [
   n1 = SetVM!() -> (then, FortNicknameVM:Nickname VM Object Reference) {is=title K2Node_FunctionEntry}
   > n2 = Set FortNicknameVM(FortNicknameVM:Nickname VM Object Reference=@n1) -> (Output_Get:Nickname VM Object Reference)
   > n3 = Set MaxCharacterLimit(MaxCharacterLimit:int=GetMaxCharacterLimit!(self:Nickname VM Object Reference=@n2) -> (ReturnValue:int=0)) -> (Output_Get:int=0)
   n4 = GetNickname!(self:Nickname VM Object Reference=@n2) -> (ReturnValue:Text)
   > n5 = Set NicknameLength(NicknameLength:int=Len(S="To String (Text)"(InText=@n4))) -> (Output_Get:int=0)
   n6 = "To String (Integer)"(InInt=@n5)
   n7 = "To String (Integer)"(InInt=@n3)
   n8 = Get ErrorTextAndCount() -> (ErrorTextAndCount:Common Text Object Reference)
   > "SetText (Text)"(self=@n8, InText="To Text (String)"(InString=Append!(A=@n6, B=/, C:String=@n7) -> (ReturnValue)))
   > SetColorAndOpacity(self=@n8, InColorAndOpacity=Get CountColor() -> (CountColor:Slate Color Structure))
  ]}
 SetVisuals {nodes [
   Inputs()
   > n1 = Cast To GridSlot!(
    execute
    Object=Get Slot!(self:Widget Object Reference=Get Prompt() -> (Prompt:Grid Panel Object Reference)) -> (Slot:Panel Slot Object Reference)
   ) -> (then, CastFailed, AsGrid Slot:Grid Slot Object Reference) {is=title K2Node_DynamicCast}
   > SetColumn!(execute, self:Grid Slot Object Reference=@n1, InColumn:int=2) -> (then)
   > n2 = SetPadding!(execute, self:Grid Slot Object Reference=@n1, InPadding:Margin Structure=Make Margin(Left=0, Top=0, Right=0, Bottom=0)) -> (then)
   n3 = Cast To GridSlot(
    execute=[@n2.then, @n1.CastFailed]
    Object=Get Slot(self=Get Block_Background() -> (Block_Background:WBP UIKit Block Background Base Object Reference))
   )
   > n4 = SetColumnSpan!(execute, self:Grid Slot Object Reference=@n3, InColumnSpan:int=2) -> (then)
   Outputs!(then:Exec=[@n4.then, @n3.CastFailed])
  ]}
 TryChangeNickname {nodes [
   n1 = TryChangeNickname!() -> (then, VM:Nickname VM Object Reference) {is=title K2Node_FunctionEntry}
   n2 = Get Text(self:Editable Text Object Reference=Get EditableText())
   > n3 = ValidateNickname(Text=@n2)
   > n4 = Branch(Condition=@n3)
   > ChangeNickname!(execute, self:Nickname VM Object Reference=@n1, Nickname:String="To String (Text)"(InText=@n2)) -> (then)
   > Return Node!(execute, Changed:bool=true)
   Return Node(execute=@n4.else, Changed=false)
  ]}
 ValidateNickname {nodes [
   n1 = ValidateNickname!() -> (then, Text:"Text (by ref)") {is=title K2Node_FunctionEntry}
   n2 = Get FortNicknameVM()
   > SetNicknameText!(execute, self:Nickname VM Object Reference=@n2, InNicknameText:"Text (by ref)"=@n1) -> (then)
   > n3 = Set NicknameLength(NicknameLength=Len(S="To String (Text)"(InText=@n1))) -> (Output_Get=0)
   > n4 = Branch(Condition="Equal (Integer)"(A=@n3))
   n5 = Get MaxCharacterLimit() -> (MaxCharacterLimit:int=0)
   n6 = "To Text (String)"(InString=Append(A="To String (Integer)"(InInt=@n3), B=/, C="To String (Integer)"(InInt=@n5)))
   n7 = Get ErrorTextAndCount()
   "SetText (Text)"(execute=[@n8.None, @n4.then], self=@n7, InText=@n6)
   > SetColorAndOpacity(self=@n7, InColorAndOpacity=Get CountColor())
   > Return Node!(execute, Valid:bool)
   n8 = Switch on ENicknamingBlockedReason(execute=@n4.else, Selection=GetNicknameError!(self:Nickname VM Object Reference=@n2) -> (ReturnValue:ENicknamingBlockedReason Enum=None))
   n9 = "SetText (Text)"(execute=[@n8.InvalidFormat, @n8.Sanitized], self=@n7, InText="LOCTEXT(\"Invalid\")")
   SetColorAndOpacity(execute=[@n10.then, @n9.then], self=@n7, InColorAndOpacity=Get ErrorColor() -> (ErrorColor:Slate Color Structure))
   > Return Node(Valid=false)
   n10 = "SetText (Text)"(execute=@n8.InvalidLength, self=@n7, InText=@n6)
  ]}
]
variables [
 CountColor {type=SlateColor, default {colorUseRule=UseColor_Specified, specifiedColor=<rgba 0.0802 0.0802 0.0887 1>}}
 ErrorColor {type=SlateColor, default {colorUseRule=UseColor_Specified, specifiedColor=<rgba 1 0 0.1992 1>}}
 MaxCharacterLimit {type=integer}
 NicknameLength {type=integer}
 OnChangedNickname {}
]
widgets {
 info {
  inheritedWidgetCount=0
  namedSlotCount=1
  parentClass=/Sidebar/Buttons/WBP_Sidebar_ButtonGeneric.WBP_Sidebar_ButtonGeneric_C
  rootWidgetClass=null
  widgetCount=3
 }
 widgets [
  HorizontalBox_2 {widgetClassPath=&HorizontalBox, widget=HorizontalBox_2}
  EditableText {
   widgetClassPath=/Script/FortniteUI.FortMobileEditableText
   bIsVariable=true
   parent=HorizontalBox_2
   widget=EditableText
   properties {
    clearKeyboardFocusOnCommit=false
    clipping=ClipToBoundsAlways
    hintText=Add Nickname
    mobileFontSizeMultiplier=1.34
    virtualKeyboardDismissAction=TextCommitOnAccept
    virtualKeyboardTrigger=OnAllFocusEvents
    widgetStyle {
     backgroundImageComposing {
      bIsDynamicallyLoaded=false
      drawAs=Border
      imageType=FullColor
      mirroring=NoMirror
      resourceName=../../../Engine/Content/Slate/Old/HyperlinkDotted.png
      resourceObject=null
      tiling=Both
      imageSize=<0 0>
      margin=<ltrb 0 0 0 0.1875>
      outlineSettings {
       bUseBrushTransparency=false
       roundingType=HalfHeightRadius
       width=0
       color {colorUseRule=UseColor_Specified, specifiedColor=<rgba 0 0 0 0>}
       cornerRadii=<0 0 0 0>
      }
      tintColor {colorUseRule=UseColor_Specified, specifiedColor=<rgba 1 1 1 1>}
      uVRegion {bIsValid=false, max=<0 0>, min=<0 0>}
     }
     backgroundImageSelected {
      bIsDynamicallyLoaded=false
      drawAs=Box
      imageType=FullColor
      mirroring=NoMirror
      resourceName=../../../Engine/Content/Slate/Common/EditableTextSelectionBackground.png
      resourceObject=null
      tiling=NoTile
      imageSize=<0 0>
      margin=<ltrb 0.25 0.25 0.25 0.25>
      outlineSettings {
       bUseBrushTransparency=false
       roundingType=HalfHeightRadius
       width=0
       color {colorUseRule=UseColor_Specified, specifiedColor=<rgba 0 0 0 0>}
       cornerRadii=<0 0 0 0>
      }
      tintColor {colorUseRule=UseColor_Specified, specifiedColor=<rgba 1 1 1 1>}
      uVRegion {bIsValid=false, max=<0 0>, min=<0 0>}
     }
     caretImage {
      bIsDynamicallyLoaded=false
      drawAs=Image
      imageType=FullColor
      mirroring=NoMirror
      resourceName=None
      resourceObject=/FortUILibrary/Systems/InputField/Materials/M_UIKit_InputFieldCaretColor
      tiling=NoTile
      imageSize=<16 16>
      margin=<ltrb 0 0 0 0>
      outlineSettings {
       bUseBrushTransparency=false
       roundingType=HalfHeightRadius
       width=0
       color {colorUseRule=UseColor_Specified, specifiedColor=<rgba 0 0 0 0>}
       cornerRadii=<0 0 0 0>
      }
      tintColor {colorUseRule=UseColor_Specified, specifiedColor=<rgba 1 1 1 1>}
      uVRegion {bIsValid=false, max=<0 0>, min=<0 0>}
     }
     colorAndOpacity {colorUseRule=UseColor_Specified, specifiedColor=<rgba 1 1 1 1>}
     font {
      bForceMonospaced=false
      bMaterialIsStencil=false
      fontMaterial=/Sidebar/Materials/Fonts/MI_Sidebar_FontColor_Primary
      fontObject=&HeadingNow
      letterSpacing=10
      monospacedWidth=1
      size=13.5
      skewAmount=0
      typefaceFontName=74 Regular
      outlineSettings {
       bApplyOutlineToDropShadows=false
       bMiteredCorners=false
       bSeparateFillAlpha=false
       outlineMaterial=/FortUILibrary/Systems/ModularButton/MaterialInstances/MI_UIKit_TextOutline
       outlineSize=0
       outlineColor=<rgba 0 0 0 1>
      }
     }
     hintColorAndOpacity {colorUseRule=UseColor_Foreground, specifiedColor=<rgba 1 0 1 1>}
     selectedBackgroundColor {colorUseRule=UseColor_Foreground, specifiedColor=<rgba 1 0 1 1>}
    }
   }
   slot {class=HorizontalBoxSlot, properties {verticalAlignment=VAlign_Center, padding=<ltrb 18 0 0 0>, size {sizeRule=Fill, value=1}}}
  }
  ErrorTextAndCount {
   widgetClassPath=&CommonTextBlock
   bIsVariable=true
   parent=HorizontalBox_2
   widget=ErrorTextAndCount
   properties {
    text=0/16
    colorAndOpacity {colorUseRule=UseColor_Specified, specifiedColor=<rgba 0.0802 0.0802 0.0887 1>}
    font {
     bForceMonospaced=false
     bMaterialIsStencil=false
     fontMaterial=null
     fontObject=&HeadingNow
     letterSpacing=10
     monospacedWidth=1
     size=13.5
     skewAmount=0
     typefaceFontName=74 Regular
     outlineSettings {
      bApplyOutlineToDropShadows=false
      bMiteredCorners=false
      bSeparateFillAlpha=false
      outlineMaterial=&MI_UIKit_TextOutline
      outlineSize=0
      outlineColor=<rgba 0 0 0 1>
     }
    }
    shadowOffset=<0 0>
   }
   slot {class=HorizontalBoxSlot, properties {horizontalAlignment=HAlign_Right, verticalAlignment=VAlign_Center, padding=<ltrb 18 0 18 0>}}
  }
 ]
}
